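# Niagara system for the FunkyFeather character intro's Zero Point FX, floating colored gem shards slowly upward from a fixed height.
# One emitter bursts and streams three spawn groups, each selecting its own mesh variants and tints.
# Age curves pop shards in and shrink them out, while curl noise, drag and system-age-scaled velocity shape their drift.
# Sprite attributes add colored glows, occasional early glints and end-of-life flashes; mesh alpha varies by platform.
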NiagaraSystem /BRCosmetics/Animation/Game/MainPlayer/CharIntros/s41/FunkyFeather/FX/NS_CharIntros_FunkyFeather_ZeroPointFX
dependencies {
 usedDataInterfaces [&NiagaraDataInterfaceCurve, &NiagaraDataInterfacePlatformSet]
 usedDynamicInputs []
 usedModules [&SystemState]
}
emitters {Shards_FloatingUp {
  emitterUpdateScript [
   &EmitterState(
    Inactive Response=::NewEnumerator1
    Life Cycle Mode=::NewEnumerator1
    Loop Behavior=::NewEnumerator1
    Loop Duration Mode=::NewEnumerator1
    UseLoopDelay=false
   )
   &SpawnBurst_Instantaneous(Spawn Count=3, Use Spawn Probability=false, UseLoopCountLimit=true)
   &SpawnRate(SpawnRate=2, Use Spawn Probability=false)
   &SpawnBurst_Instantaneous(Spawn Count=6, Spawn Group=1, Use Spawn Probability=false, UseLoopCountLimit=true) {module=SpawnBurst_Instantaneous001}
   &SpawnRate(SpawnGroup=1, SpawnRate=3.5, Use Spawn Probability=false) {module=SpawnRate001}
   &SpawnBurst_Instantaneous(Spawn Count=2, Spawn Group=2, Use Spawn Probability=false, UseLoopCountLimit=true) {module=SpawnBurst_Instantaneous002}
   &SpawnRate(SpawnGroup=2, Use Spawn Probability=false) {module=SpawnRate002}
  ]
  particleSpawnScript [
   /SetVariables(Particles.PerParticleRandom:NiagaraFloat=&RandomRangeFloat(Minimum=0, Maximum=1)) {isSetParameters=true}
   &InitializeParticle(
    Lifetime=&RandomRangeFloat(Minimum=14, Maximum=16)
    Mesh Scale Mode=::NewEnumerator3
    Mesh Uniform Scale=&Lerp_Float(A=0.12, B=0.225, Alpha=$Particles.PerParticleRandom)
    Sprite Rotation Mode=::NewEnumerator2
    Sprite Size=<30 30>
    Sprite Size Mode=::NewEnumerator4
    UsePositionOffset=false
    Write Material Random=true
   )
   &ShapeLocation(
    Additional Quaternion Rotation=false
    Apply To Particle Position=::NewEnumerator1
    Disc Coverage=0.2
    MaskSpawnGroup=false
    Offset=<35 0 0>
    Offset Coordinate Space=::Local
    Offset Mode=::NewEnumerator0
    Override Seed=false
    Ring Radius=50
    Rotation Mode=::NewEnumerator3
    Rotation Quaternion=&MultiplyQuaternion(
     Quaternion A=&MakeQuaternion(
      Quaternion From=1
      Axis=<0 1 0>
      Angle=0
      Yaw=-25
      From=?
      To=?
      Stabilization Mode:ENiagara_OrientVectorCount=::NewEnumerator0
      Side Direction:Vector3f=?
      Up Direction:Vector3f=?
      Angle Type=::NewEnumerator0
      Coordinate Space=::World
      Normalize:NiagaraBool=true
     )
     Quaternion B=&MakeQuaternion(
      Quaternion From=0
      Axis=<0 1 0>
      Angle=0
      From=?
      To=?
      Stabilization Mode:ENiagara_OrientVectorCount=::NewEnumerator0
      Side Direction:Vector3f=?
      Up Direction:Vector3f=?
      Angle Type=::NewEnumerator0
      Coordinate Space=::World
      Normalize:NiagaraBool=true
     )
    )
    Shape Primitive=::NewEnumerator4
    U Distribution=0.2
    UseOwnerScale=false
   )
   /SetVariables(Particles.Position=&MakeVector(
     X=&MakeFloatFromVector(Vector=Linked() {linkedVariable {name=Particles.ShapeLocation.ShapePosition, type=NiagaraPosition}}, Channel=::NewEnumerator0)
     Y=&MakeFloatFromVector(Vector=Linked() {linkedVariable {name=Particles.ShapeLocation.ShapePosition, type=NiagaraPosition}}, Channel=::NewEnumerator1)
     Z=75
    )) {isSetParameters=true}
   /SetVariables(Particles.Direction:Vector3f=&NormalizeVector(
     Vector To Normalize=&Subtract_Vector(
      A=Linked() {linkedVariable {name=Particles.ShapeLocation.ShapePosition, type=NiagaraPosition}}
      B=Linked() {linkedVariable {name=Engine.Emitter.SimulationPosition, type=NiagaraPosition}}
     )
     Safe Normalize=true
     Safe Fallback Threshold=0
     Safe Fallback Vector=<0 0 0.001>
    )) {isSetParameters=true}
   &InitialMeshOrientation(Rotation=<-0.25 0 0>, Use Rotation Vector=false)
   /SetVariables(Particles.SpawnGroup=$Emitter.SpawnGroup) {isSetParameters=true}
   /SetVariables(Particles.MeshIndex=&MergeHelper_IntSwitch(
     0=&RandomRangeInt(Minimum=0, Maximum=2)
     1=&RandomRangeInt(Minimum=3, Maximum=5)
     2=&RandomRangeInt(Minimum=6, Maximum=9)
     3=?
     4=?
     5=?
     Index=$Particles.SpawnGroup
     Num=::NewEnumerator1
    )) {isSetParameters=true}
   &AddVelocity(
    Rotation Coordinate Space=::World
    Velocity=&Lerp_Vector(
     A=&RandomRangeVector(Minimum=<0 0 8>, Maximum=<0 0 10>)
     B=&RandomRangeVector(Minimum=<0 0 3.25>, Maximum=<0 0 5.5>)
     Alpha=$Particles.PerParticleRandom
    )
   )
   &AddVelocity(
    Rotation Coordinate Space=::World
    Velocity=&Multiply_VectorByFloat(Vector=$Particles.Direction, Float=&RandomRangeFloat(Minimum=-0.3, Maximum=2))
   ) {module=AddVelocity001}
  ]
  particleUpdateScript [
   &ParticleState()
   /SetVariables(Particles.GemLifetimeControl:NiagaraFloat=&Multiply_Float(
     A=&FloatFromCurve(
      FloatCurve=DataInterface() {propertyValues {Curve {keys [
          <at 0 0 RCIM_Cubic>
          {arriveTangent=6.125, interpMode=RCIM_Cubic, leaveTangent=6.125, time=0.1, value=0.15}
          {arriveTangent=5.6667, interpMode=RCIM_Cubic, leaveTangent=5.6667, time=0.2, value=1.225}
          {arriveTangentWeight=0.0167, interpMode=RCIM_Cubic, leaveTangentWeight=1.0858, tangentMode=RCTM_User, time=0.25, value=1}
         ]}}}
      CurveIndex=$Particles.Age
     )
     B=&FloatFromCurve(FloatCurve=DataInterface() {propertyValues {Curve {keys [
          {arriveTangentWeight=1.1185, interpMode=RCIM_Cubic, leaveTangentWeight=0.0501, tangentMode=RCTM_User, time=0.9636, value=1}
          {arriveTangent=-6.8846, interpMode=RCIM_Cubic, leaveTangent=-6.8846, tangentMode=RCTM_User, time=0.9773, value=0.9}
          {interpMode=RCIM_Cubic, leaveTangent=0.0104, time=1, value=0}
         ]}}})
    )) {isSetParameters=true}
   &UpdateMeshOrientation(Rotation Rate=&Lerp_Float(
     A=&FloatFromCurve(FloatCurve=DataInterface() {propertyValues {Curve {keys [<at 0 0.5 RCIM_Cubic>,<at 0.8 0.25>]}}}, Scale Curve=0.3)
     B=&FloatFromCurve(FloatCurve=DataInterface() {propertyValues {Curve {keys [<at 0 0.2 RCIM_Cubic>,<at 1 0.065>]}}}, Scale Curve=0.5)
     Alpha=$Particles.PerParticleRandom
    )) {module=UpdateMeshOrientation001}
   &CurlNoiseForce(
    Bias Noise=false
    Noise Frequency=0.5
    Noise Quality / Cost=::NewEnumerator0
    Pan Noise=true
    Pan Noise Field=&VectorFromFloat(Value=0.1)
   )
   &ScaleMeshSize(Scale Factor=&VectorFromFloat(Value=$Particles.GemLifetimeControl))
   &DynamicMaterialParameters(
    Index 0 Param 1=1
    Index 3 Param 1=&MakeFloatFromInt(Int=$Particles.VisibilityTag)
    Index 3 Param 2=&MakeFloatFromVector(Vector=$Particles.InterpolateOverTime.InterpolateOverTime_V3_.Moving Average, Channel=::NewEnumerator0)
    Index 3 Param 3=&MakeFloatFromVector(Vector=$Particles.InterpolateOverTime.InterpolateOverTime_V3_.Moving Average, Channel=::NewEnumerator1)
    Index 3 Param 4=&MakeFloatFromVector(Vector=$Particles.InterpolateOverTime.InterpolateOverTime_V3_.Moving Average, Channel=::NewEnumerator2)
    Module.Index0Float0Write=true
    Module.Index0Float1Write=true
    Module.Index0Float2Write=true
    Module.Index0Float3Write=true
    Module.Index3Float0Write=true
    Module.Index3Float1Write=true
    Module.Index3Float2Write=true
    Module.Index3Float3Write=true
    Write Parameter Index 3=::NewEnumerator1
   )
   &FN_Mesh_Attributes(
    Color=&MultiplyLinearColor(
     Linear Color=&MergeHelper_ColorSwitch(
      0=<rgba 0.9 0.045 0.4911 0.3>
      1=<rgba 0.0585 0.0585 0.9 0.75>
      2=&MergeHelper_ColorSwitch(
       0=<rgba 0.5304 0 0.3321 0.6>
       1=<rgba 0.7 0.1867 0 0.6>
       2=<rgba 0.051 0.6752 0.8 0.6>
       3=<rgba 0.0426 0.1973 0.7565 0.6>
       Index=/Niagara/DynamicInputs/Subtract/Subtract_Integer(A:NiagaraInt32=$Particles.MeshIndex, B:NiagaraInt32=6)
       Num=::NewEnumerator2
      )
      3=?
      Index=$Particles.SpawnGroup
      Num=::NewEnumerator1
     )
     Scale Factor=&MakeVector4FromVectorAndFloat(
      "Vector (RGB)"=<1 1 1>
      "Float (Alpha)"=&FN_PickFloatDependOnPlatform(TargetPlatform=DataInterface() {propertyValues {Platforms {qualityLevelMask=-29}}}, TargetPlatformActive=0.25, TargetPlatformNoActive=1)
     )
    )
    Create Color Attribute=true
    Module Output=::NewEnumerator1
   )
   &FN_Sprite_Attributes(
    Color=&MultiplyLinearColor(
     Linear Color=&MergeHelper_ColorSwitch(
      0=<rgba 0.3 0 0.3 0.25>
      1=<rgba 0.1 0 0.3 0.3>
      2=&MergeHelper_ColorSwitch(
       0=<rgba 0.5304 0 0.3321 0.165>
       1=<rgba 0.7 0.0415 0 0.135>
       2=<rgba 0.051 0.8 0.3897 0.06>
       3=<rgba 0.0426 0.1616 0.7565 0.235>
       Index=&Subtract_Integer(A:NiagaraInt32=$Particles.MeshIndex, B:NiagaraInt32=6)
       Num=::NewEnumerator2
      )
      3=?
      Index=$Particles.SpawnGroup
      Num=::NewEnumerator1
     )
     Scale Factor=<1 1 1 0.15>
    )
    Create Color Attribute=true
    Create Dynamic Material Parameter Attributes=true
    Create SpriteSize Attribute=true
    Index 0 Vector 4:Vector4f=<0 1 0 0>
    Module Output:ENiagara_AttributeSamplingApplyOutput=::NewEnumerator1
    SpriteSize=&Multiply_Vector2D_ByFloat(Float=$Particles.GemLifetimeControl, Vector2D=<20 20>)
    Write Parameter Index 0:ENiagara_DynamicMaterialParametersInputMode=::NewEnumerator4
   )
   &FN_Sprite_Attributes(
    Camera Offset=35
    Color=&MakeLinearColorFromVectorAndFloat(
     "Vector (RGB)"=&MakeVectorFromLinearColorRGB(Color=<rgba 5 2.3021 0.625 1>)
     "Float (Alpha)"=&Lerp_Float(
      A=1
      B=&RandomRangeFloat(Minimum=0, Maximum=1, Evaluation Type=::NewEnumerator1)
      Alpha=&FloatFromCurve(FloatCurve=DataInterface() {propertyValues {Curve {keys [<at 0.4 1>,<at 1 0>]}}})
     )
    )
    Create Camera Offset Attribute=true
    Create Color Attribute=true
    Create Position Attribute=true
    Create SpriteSize Attribute=true
    Module Output:ENiagara_AttributeSamplingApplyOutput=::NewEnumerator1
    Position:NiagaraPosition=&AddVectorToPosition(Position=$Particles.Position, Vector=<0.5 0 1>)
    SpriteSize=&Vector2DFromFloat(Value=&Multiply_Float(
      A=&RandomRangeFloat(Minimum=11, Maximum=13)
      B=&FloatFromCurve(
       FloatCurve=DataInterface() {propertyValues {Curve {keys [
           <at 0 0 RCIM_Constant>
           {arriveTangent=11.6057, interpMode=RCIM_Cubic, leaveTangent=11.6057, tangentMode=RCTM_User, time=0.3, value=0}
           {arriveTangent=-16.656, interpMode=RCIM_Cubic, leaveTangent=-16.656, tangentMode=RCTM_User, time=0.4, value=1}
           <at 0.6 0>
          ]}}}
       CurveIndex=&ClampFloat(
        Float=&Add_Float(A=&RandomRangeFloat(Minimum=-2.5, Maximum=-3), B=$Particles.Age)
        Max=5
       )
       Scale Curve=&MakeCustomFloatFromBool(
        Bool=&SetBoolByFloatComparison(A=&RandomRangeFloat(Minimum=0, Maximum=1), B=0.15)
        True Float=1
        False Float=0
       )
      )
     ))
   ) {module=FN_Sprite_Attributes001}
   &FN_Sprite_Attributes(
    Color=&MakeLinearColorFromVectorAndFloat(
     "Vector (RGB)"=&MakeVectorFromLinearColorRGB(Color=<rgba 5 2.3021 0.625 1>)
     "Float (Alpha)"=&Lerp_Float(
      A=1
      B=&RandomRangeFloat(Minimum=0, Maximum=1, Evaluation Type=::NewEnumerator1)
      Alpha=&FloatFromCurve(FloatCurve=DataInterface() {propertyValues {Curve {keys [<at 0.4 1>,<at 1 0>]}}})
     )
    )
    Create Color Attribute=true
    Create SpriteSize Attribute=true
    Module Output:ENiagara_AttributeSamplingApplyOutput=::NewEnumerator1
    SpriteSize=&Vector2DFromFloat(Value=&Multiply_Float(
      A=&RandomRangeFloat(Minimum=12, Maximum=14)
      B=&FloatFromCurve(FloatCurve=DataInterface() {propertyValues {Curve {keys [
           {arriveTangent=36.5951, interpMode=RCIM_Cubic, leaveTangent=36.5951, tangentMode=RCTM_User, time=0.9773, value=0}
           {arriveTangent=-17.2427, interpMode=RCIM_Cubic, leaveTangent=-17.2427, tangentMode=RCTM_User, time=0.9864, value=1}
           <at 1 0>
          ]}}})
     ))
   ) {module=FN_Sprite_Attributes002}
   &AccelerationForce(Acceleration=&MakeVector(X=0, Y=0, Z=&Lerp_Float(A=7.5, B=2, Alpha=$Particles.PerParticleRandom)))
   &Drag(Use Linear Drag=true, Use Rotational Drag=true)
   &ScaleVelocity(
    Coordinate Space:ENiagaraCoordinateSpace=::World
    Velocity Scale=&MakeVector(
     X=1
     Y=1
     Z=&FloatFromCurve(
      FloatCurve=DataInterface() {propertyValues {Curve {keys [
          {arriveTangentWeight=0.5513, leaveTangentWeight=0.2205, time=2.5, value=6}
          <at 2.8 3.3>
          {arriveTangent=-9.417, arriveTangentWeight=0.0111, leaveTangent=-9.417, time=3.1667, value=1}
         ]}}}
      CurveIndex=$System.Age
     )
    )
   )
   &SolveForcesAndVelocity(Clamp Velocity=false, Limit Acceleration=false)
  ]
  rendererClasses [&NiagaraMeshRendererProperties, &NiagaraSpriteRendererProperties]
  renderers [
   {rendererClass=&NiagaraMeshRendererProperties, rendererIndex=0}
   {rendererClass=&NiagaraMeshRendererProperties, rendererIndex=1}
   {rendererClass=&NiagaraSpriteRendererProperties, rendererIndex=2}
   {rendererClass=&NiagaraSpriteRendererProperties, rendererIndex=3}
   {rendererClass=&NiagaraSpriteRendererProperties, rendererIndex=4}
  ]
 }}
summary {
 systemName=NS_CharIntros_FunkyFeather_ZeroPointFX
 emitters [{
   emitterName=Shards_FloatingUp
   rendererClasses [&NiagaraMeshRendererProperties, &NiagaraSpriteRendererProperties]
  }]
 userVariables []
}
